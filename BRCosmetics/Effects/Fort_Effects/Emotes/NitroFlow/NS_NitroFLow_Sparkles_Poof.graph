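# Niagara sparkle poof for the NitroFlow emote, built from a single BurstEmbers emitter that continuously rate-spawns short-lived embers drawn by two sprite renderers.
# User.IsAlt blends the spawn color from HDR blue to orange.
# Particles spawn within a small sphere with randomized velocity, then slow under drag and light gravity.
# A size curve, randomly offset waveform flicker and speed-based stretching make them twinkle and streak.
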NiagaraSystem /BRCosmetics/Effects/Fort_Effects/Emotes/NitroFlow/NS_NitroFLow_Sparkles_Poof
dependencies {usedDataInterfaces [&NiagaraDataInterfaceCurve], usedDynamicInputs [], usedModules [&SystemState]}
emitters {BurstEmbers {
  emitterUpdateScript [
   &EmitterState()
   &SpawnRate(SpawnRate=100, Use Spawn Probability=false)
  ]
  particleSpawnScript [
   &InitializeParticle(
    Color=&LerpLinearColors(StartColor=<rgba 1.0997 2.3998 5 1>, EndColor=<rgba 5 2.0747 1.0997 1>, LerpFactor=$User.IsAlt)
    Lifetime Max=0.35
    Lifetime Mode=::NewEnumerator1
    Sprite Size=&Multiply_Vector2D_ByFloat(Float=2, Vector2D=&RandomRangeVector2D(Minimum=<4 5>, Maximum=<4 5>))
    Sprite Size Mode=::NewEnumerator4
    UsePositionOffset=false
    Write Material Random=true
   ) {module=InitializeParticle001}
   &AddVelocity(Velocity=&RandomVector(Vector Scale=20))
   &ShapeLocation(MaskSpawnGroup=false, Override Seed=false, Sphere Radius=10, UseOwnerScale=true)
   &AddVelocity(
    Module.Use Falloff Distance=false
    Override Seed=false
    Velocity Mode=::NewEnumerator1
    Velocity Speed=&RandomRangeFloat(Minimum=10, Maximum=100)
   ) {module=AddVelocity001}
   &ApplyInitialForces()
   &CameraOffset(Camera Offset Amount=5)
  ]
  particleUpdateScript [
   &ParticleState()
   &ScaleSpriteSize(
    Scale Sprite Size Mode=::NewEnumerator0
    Uniform Scale Factor=&MakeFloatFromVector(
     Vector=&VectorFromFloat(Value=&Waveform(
       "Waveform [1]"=::NewEnumerator12
       "[1] Input Position"=&RandomRangeFloat(Minimum=0, Maximum=1000)
       "[1] Frequency"=25
       "[1] Amplitude Min/Max"=<0.2 1>
      ))
     Channel=::NewEnumerator0
    )
   ) {module=ScaleSpriteSize001}
   &ScaleSpriteSize(Scale Factor=&Vector2DFromFloat(Value=&FloatFromCurve(FloatCurve=DataInterface() {propertyValues {Curve {keys [
          {arriveTangent=11.6538, interpMode=RCIM_Cubic, time=0, value=0.0211}
          {arriveTangent=0.3289, interpMode=RCIM_Cubic, leaveTangent=0.3289, time=0.125, value=1.5}
          <at 1 0.35 RCIM_Cubic>
         ]}}})))
   &Drag(Drag=5, Use Linear Drag=true, Use Rotational Drag=false)
   &GravityForce(Gravity=<0 0 -100>)
   &SolveForcesAndVelocity(Clamp Velocity=false, Limit Acceleration=false)
   &ScaleSpriteSizeBySpeed(Max Scale Factor=<1 4>, Source Velocity:Vector3f=&Multiply_Vector(A=<0 0.01 1>, B=$Particles.Velocity))
   &ScaleColor(
    Color Value To Scale=$Particles.Color
    Scale Alpha=&FloatFromCurve(FloatCurve=DataInterface() {propertyValues {Curve {keys [<at 0 0 RCIM_Cubic>,<at 0.1 1 RCIM_Cubic>,<at 1 0>]}}})
    Scale RGB=&VectorFromFloat(Value=1)
    ScaleA=false
    ScaleRGB=true
   ) {module=ScaleColor001}
  ]
  rendererClasses [&NiagaraSpriteRendererProperties]
  renderers [
   {rendererClass=&NiagaraSpriteRendererProperties}
   {rendererClass=&NiagaraSpriteRendererProperties, rendererIndex=1}
  ]
 }}
summary {
 systemName=NS_NitroFLow_Sparkles_Poof
 emitters [{emitterName=BurstEmbers, rendererClasses [&NiagaraSpriteRendererProperties]}]
 userVariables [
  User.bIsFrontEndPreview {type=NiagaraBool}
  User.IsAlt {type=NiagaraFloat, defaultValue=0, description=""}
  User.SkeletalMesh {
   type=NiagaraDataInterfaceSkeletalMesh
   defaultValue {
    struct=NiagaraExt_VariableValue_DataInterface
    value {dataInterface=&self:NiagaraDataInterfaceSkeletalMesh_0}
   }
  }
 ]
}
