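# Niagara flare effect for the Nitro Flow emote: a short-lived sprite burst tinted blue, or warm orange as User.IsAlt goes to 1.
# A single emitter nudges the sprite with a camera offset, spins it and scales it through a quick pop-and-shrink curve.
# It also derives an enlarged halo size and a faint-alpha color, apparently to layer a soft glow through its two sprite renderers.
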NiagaraSystem /BRCosmetics/Effects/Fort_Effects/Emotes/NitroFlow/NS_Nitro_Flow_Flare
dependencies {usedDataInterfaces [&NiagaraDataInterfaceCurve], usedDynamicInputs [], usedModules [&SystemState]}
emitters {SingleLoopingParticle {
  emitterUpdateScript [
   &EmitterState()
   &SpawnBurst_Instantaneous(Use Spawn Probability=false, UseLoopCountLimit=false)
  ]
  particleSpawnScript [
   &InitializeParticle(
    Color=&LerpLinearColors(StartColor=<rgba 1.59 2.6698 5 1>, EndColor=<rgba 4 2.8613 0.584 1>, LerpFactor=$User.IsAlt)
    Lifetime=0.5
    Sprite Rotation Angle=-5
    Sprite Rotation Mode=::NewEnumerator0
    Sprite Size Mode=::NewEnumerator3
    Uniform Sprite Size=35
    UsePositionOffset=false
    Write Material Random=true
   )
   &CameraOffset(Camera Offset Amount=20)
  ]
  particleUpdateScript [
   &ParticleState()
   &SpriteRotationRate(Rotation Rate=30)
   &ScaleSpriteSize(Uniform Curve Sprite Scale=DataInterface() {propertyValues {Curve {keys [
        {arriveTangent=9.0725, interpMode=RCIM_Cubic, leaveTangent=9.0725, tangentMode=RCTM_User, time=0, value=0}
        {arriveTangent=-2.5249, interpMode=RCIM_Cubic, leaveTangent=-2.5249, tangentMode=RCTM_User, time=0.1408, value=0.9847}
        {arriveTangent=-0.0405, interpMode=RCIM_Cubic, leaveTangent=-0.0405, tangentMode=RCTM_User, time=1, value=0}
       ]}}})
   &SolveForcesAndVelocity(Clamp Velocity=false, Limit Acceleration=false)
   &ScaleColor(ScaleA=true, ScaleRGB=true)
   /SetVariables(
    Particles.Halo Size:Vector2f=&Multiply_Vector2D_ByFloat(Float=3, Vector2D=$Particles.SpriteSize)
    Particles.New Linear Color:LinearColor=&MultiplyLinearColor(Linear Color=$Particles.Color, Scale Factor=<1 1 1 0.05>)
   ) {isSetParameters=true}
  ]
  rendererClasses [&NiagaraSpriteRendererProperties]
  renderers [
   {rendererClass=&NiagaraSpriteRendererProperties}
   {rendererClass=&NiagaraSpriteRendererProperties, rendererIndex=1}
  ]
 }}
summary {
 systemName=NS_Nitro_Flow_Flare
 emitters [{emitterName=SingleLoopingParticle, rendererClasses [&NiagaraSpriteRendererProperties]}]
 userVariables [User.IsAlt {type=NiagaraFloat, defaultValue=0, description=""}]
}
